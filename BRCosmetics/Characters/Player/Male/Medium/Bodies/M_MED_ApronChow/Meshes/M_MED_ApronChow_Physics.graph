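# Physics asset for the male medium ApronChow cosmetic body that sets up simulated motion for its shirt hem.
# Eight dyn_shirt capsules around the front, back and sides hinge on tiny base_shirt anchors, with narrow swing limits and no twist.
# The anchors join the pelvis with free angular constraints, as do large thigh capsules that form the leg volumes.
# Extra angularly free constraints link the dyn capsules to the back and right_04 ones.
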PhysicsAsset /BRCosmetics/Characters/Player/Male/Medium/Bodies/M_MED_ApronChow/Meshes/M_MED_ApronChow_Physics
bodies {
 base_shirt_bk_01 [{shapeType=Capsule, length=0.5, radius=0.5, shapeName=base_shirt_bk_01_capsule, rotation=<rot 0 -90 0>}]
 base_shirt_fr_01 [{shapeType=Capsule, length=0.5, radius=0.5, shapeName=base_shirt_fr_01_capsule, rotation=<rot 0 -90 0>}]
 base_shirt_l_01 [{shapeType=Capsule, length=0.5, radius=0.5, shapeName=base_shirt_l_01_capsule, rotation=<rot 0 -90 0>}]
 base_shirt_l_02 [{shapeType=Capsule, length=0.5, radius=0.5, shapeName=base_shirt_l_02_capsule, rotation=<rot 0 -90 0>}]
 base_shirt_l_04 [{shapeType=Capsule, length=0.5, radius=0.5, shapeName=base_shirt_l_04_capsule, rotation=<rot 0 -90 0>}]
 base_shirt_r_01 [{shapeType=Capsule, length=0.5, radius=0.5, shapeName=base_shirt_r_01_capsule, rotation=<rot 0 -90 0>}]
 base_shirt_r_02 [{shapeType=Capsule, length=0.5, radius=0.5, shapeName=base_shirt_r_02_capsule, rotation=<rot 0 -90 0>}]
 base_shirt_r_04 [{shapeType=Capsule, length=0.5, radius=0.5, shapeName=base_shirt_r_04_capsule, rotation=<rot 0 -90 0>}]
 dyn_shirt_bk_01 [{shapeType=Capsule, length=11, radius=3.3, shapeName=dyn_shirt_bk_01_capsule, center=<5 -2 0>}]
 dyn_shirt_fr_01 [{shapeType=Capsule, length=11, radius=3.3, shapeName=dyn_shirt_fr_01_capsule, center=<5 -3 0>}]
 dyn_shirt_l_01 [{shapeType=Capsule, length=11, radius=3.3, shapeName=dyn_shirt_l_01_capsule, center=<5 -4.2 0>}]
 dyn_shirt_l_02 [{shapeType=Capsule, length=11, radius=3.3, shapeName=dyn_shirt_l_02_capsule, center=<5 -3.5 0>}]
 dyn_shirt_l_04 [{shapeType=Capsule, length=11, radius=3.3, shapeName=dyn_shirt_l_04_capsule, center=<5.5 -4.8 0>}]
 dyn_shirt_r_01 [{shapeType=Capsule, length=11, radius=3.3, shapeName=dyn_shirt_r_01_capsule, center=<5 -3 0>}]
 dyn_shirt_r_02 [{shapeType=Capsule, length=11, radius=3.3, shapeName=dyn_shirt_r_02_capsule, center=<5 -4.8 0>}]
 dyn_shirt_r_04 [{shapeType=Capsule, length=11, radius=3.3, shapeName=dyn_shirt_r_04_capsule, center=<5 -4 0>}]
 pelvis [{shapeType=Capsule, length=10, radius=11.2, shapeName=pelvis_capsule, center=<-0.8019 -1.2953 0>}]
 thigh_l [{shapeType=Capsule, length=40, radius=9.5, shapeName=thigh_l_capsule, center=<-16.8032 0.4502 0.5528>, rotation=<rot 0 -91.2523 -87.8053>}]
 thigh_r [{shapeType=Capsule, length=40, radius=9.5, shapeName=thigh_r_capsule, center=<16.8032 -0.4502 -0.5528>, rotation=<rot 0 -91.2523 92.1947>}]
}
constraints [
 {bone1Name=dyn_shirt_l_01, bone2Name=base_shirt_l_01, swing1LimitDegrees=20, swing2LimitDegrees=2, twistLimitDegrees=0}
 {bone1Name=dyn_shirt_r_01, bone2Name=base_shirt_r_01, swing1LimitDegrees=20, swing2LimitDegrees=2, twistLimitDegrees=0}
 {bone1Name=dyn_shirt_fr_01, bone2Name=base_shirt_fr_01, swing1LimitDegrees=20, swing2LimitDegrees=2, twistLimitDegrees=0}
 {bone1Name=dyn_shirt_bk_01, bone2Name=base_shirt_bk_01, swing1LimitDegrees=25, swing2LimitDegrees=2, twistLimitDegrees=0}
 {bone1Name=dyn_shirt_r_02, bone2Name=base_shirt_r_02, swing1LimitDegrees=20, swing2LimitDegrees=2, twistLimitDegrees=0}
 {bone1Name=dyn_shirt_l_02, bone2Name=base_shirt_l_02, swing1LimitDegrees=20, swing2LimitDegrees=2, twistLimitDegrees=0}
 {bone1Name=dyn_shirt_l_04, bone2Name=base_shirt_l_04, swing1LimitDegrees=30, swing2LimitDegrees=3, twistLimitDegrees=0}
 {bone1Name=dyn_shirt_r_04, bone2Name=base_shirt_r_04, swing1LimitDegrees=20, swing2LimitDegrees=2, twistLimitDegrees=0}
 {
  bone1Name=base_shirt_bk_01
  bone2Name=pelvis
  swing1LimitDegrees=45
  swing1Motion=Free
  swing2LimitDegrees=45
  swing2Motion=Free
  twistLimitDegrees=45
  twistMotion=Free
 }
 {
  bone1Name=base_shirt_fr_01
  bone2Name=pelvis
  swing1LimitDegrees=45
  swing1Motion=Free
  swing2LimitDegrees=45
  swing2Motion=Free
  twistLimitDegrees=45
  twistMotion=Free
 }
 {
  bone1Name=base_shirt_l_01
  bone2Name=pelvis
  swing1LimitDegrees=45
  swing1Motion=Free
  swing2LimitDegrees=45
  swing2Motion=Free
  twistLimitDegrees=45
  twistMotion=Free
 }
 {
  bone1Name=base_shirt_l_02
  bone2Name=pelvis
  swing1LimitDegrees=45
  swing1Motion=Free
  swing2LimitDegrees=45
  swing2Motion=Free
  twistLimitDegrees=45
  twistMotion=Free
 }
 {
  bone1Name=base_shirt_l_04
  bone2Name=pelvis
  swing1LimitDegrees=45
  swing1Motion=Free
  swing2LimitDegrees=45
  swing2Motion=Free
  twistLimitDegrees=45
  twistMotion=Free
 }
 {
  bone1Name=base_shirt_r_01
  bone2Name=pelvis
  swing1LimitDegrees=45
  swing1Motion=Free
  swing2LimitDegrees=45
  swing2Motion=Free
  twistLimitDegrees=45
  twistMotion=Free
 }
 {
  bone1Name=base_shirt_r_02
  bone2Name=pelvis
  swing1LimitDegrees=45
  swing1Motion=Free
  swing2LimitDegrees=45
  swing2Motion=Free
  twistLimitDegrees=45
  twistMotion=Free
 }
 {
  bone1Name=base_shirt_r_04
  bone2Name=pelvis
  swing1LimitDegrees=45
  swing1Motion=Free
  swing2LimitDegrees=45
  swing2Motion=Free
  twistLimitDegrees=45
  twistMotion=Free
 }
 {
  bone1Name=thigh_l
  bone2Name=pelvis
  swing1LimitDegrees=45
  swing1Motion=Free
  swing2LimitDegrees=45
  swing2Motion=Free
  twistLimitDegrees=45
  twistMotion=Free
 }
 {
  bone1Name=thigh_r
  bone2Name=pelvis
  swing1LimitDegrees=45
  swing1Motion=Free
  swing2LimitDegrees=45
  swing2Motion=Free
  twistLimitDegrees=45
  twistMotion=Free
 }
 {
  bone1Name=dyn_shirt_bk_01
  bone2Name=dyn_shirt_r_04
  swing1LimitDegrees=45
  swing1Motion=Free
  swing2LimitDegrees=45
  swing2Motion=Free
  twistLimitDegrees=45
  twistMotion=Free
 }
 {
  bone1Name=dyn_shirt_fr_01
  bone2Name=dyn_shirt_r_04
  swing1LimitDegrees=45
  swing1Motion=Free
  swing2LimitDegrees=45
  swing2Motion=Free
  twistLimitDegrees=45
  twistMotion=Free
 }
 {
  bone1Name=dyn_shirt_l_01
  bone2Name=dyn_shirt_r_04
  swing1LimitDegrees=45
  swing1Motion=Free
  swing2LimitDegrees=45
  swing2Motion=Free
  twistLimitDegrees=45
  twistMotion=Free
 }
 {
  bone1Name=dyn_shirt_l_02
  bone2Name=dyn_shirt_r_04
  swing1LimitDegrees=45
  swing1Motion=Free
  swing2LimitDegrees=45
  swing2Motion=Free
  twistLimitDegrees=45
  twistMotion=Free
 }
 {
  bone1Name=dyn_shirt_l_04
  bone2Name=dyn_shirt_r_04
  swing1LimitDegrees=45
  swing1Motion=Free
  swing2LimitDegrees=45
  swing2Motion=Free
  twistLimitDegrees=45
  twistMotion=Free
 }
 {
  bone1Name=dyn_shirt_r_01
  bone2Name=dyn_shirt_r_04
  swing1LimitDegrees=45
  swing1Motion=Free
  swing2LimitDegrees=45
  swing2Motion=Free
  twistLimitDegrees=45
  twistMotion=Free
 }
 {
  bone1Name=dyn_shirt_r_02
  bone2Name=dyn_shirt_r_04
  swing1LimitDegrees=45
  swing1Motion=Free
  swing2LimitDegrees=45
  swing2Motion=Free
  twistLimitDegrees=45
  twistMotion=Free
 }
 {
  bone1Name=dyn_shirt_r_04
  bone2Name=dyn_shirt_bk_01
  swing1LimitDegrees=45
  swing1Motion=Free
  swing2LimitDegrees=45
  swing2Motion=Free
  twistLimitDegrees=45
  twistMotion=Free
 }
 {
  bone1Name=dyn_shirt_r_02
  bone2Name=dyn_shirt_bk_01
  swing1LimitDegrees=45
  swing1Motion=Free
  swing2LimitDegrees=45
  swing2Motion=Free
  twistLimitDegrees=45
  twistMotion=Free
 }
 {
  bone1Name=dyn_shirt_r_01
  bone2Name=dyn_shirt_bk_01
  swing1LimitDegrees=45
  swing1Motion=Free
  swing2LimitDegrees=45
  swing2Motion=Free
  twistLimitDegrees=45
  twistMotion=Free
 }
 {
  bone1Name=dyn_shirt_l_04
  bone2Name=dyn_shirt_bk_01
  swing1LimitDegrees=45
  swing1Motion=Free
  swing2LimitDegrees=45
  swing2Motion=Free
  twistLimitDegrees=45
  twistMotion=Free
 }
 {
  bone1Name=dyn_shirt_l_02
  bone2Name=dyn_shirt_bk_01
  swing1LimitDegrees=45
  swing1Motion=Free
  swing2LimitDegrees=45
  swing2Motion=Free
  twistLimitDegrees=45
  twistMotion=Free
 }
 {
  bone1Name=dyn_shirt_l_01
  bone2Name=dyn_shirt_bk_01
  swing1LimitDegrees=45
  swing1Motion=Free
  swing2LimitDegrees=45
  swing2Motion=Free
  twistLimitDegrees=45
  twistMotion=Free
 }
 {
  bone1Name=dyn_shirt_fr_01
  bone2Name=dyn_shirt_bk_01
  swing1LimitDegrees=45
  swing1Motion=Free
  swing2LimitDegrees=45
  swing2Motion=Free
  twistLimitDegrees=45
  twistMotion=Free
 }
]
