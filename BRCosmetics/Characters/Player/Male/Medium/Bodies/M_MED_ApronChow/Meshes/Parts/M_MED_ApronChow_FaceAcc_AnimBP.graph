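# Post-process style animation blueprint for the ApronChow male face accessory, driving its mesh from the parent character's pose.
# AnimGraph copies the attached parent pose into a Control Rig, then springs four hat bones with head-relative Anim Dynamics.
# Weapon changes set rifle, pistol and melee flags, which combine each update with crouching and targeting into rig inputs.
# A parameter asset also computes gravity and rigid-body settings, though AnimGraph reads neither.
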AnimBlueprint /BRCosmetics/Characters/Player/Male/Medium/Bodies/M_MED_ApronChow/Meshes/Parts/M_MED_ApronChow_FaceAcc_AnimBP
target_skeleton=&/M_MED_ApronChow_FaceAcc_Skeleton
graphs [
 AnimGraph {nodes [
   n1 = Control Rig!(
    Source=Copy Pose From Mesh() {bUseAttachedParent=true}
    Alpha
    Is Crouch:double=Get crouchMoving() -> (crouchMoving:double)
    Rifle:double=Get crouchMoving_Rifle() -> (crouchMoving_Rifle:double)
    crouchMoving_Rifle:double=Get crouchMoving_Rifle()
    ISCrouching:bool=Get bIsCrouching()
    IsTragetting:bool=Get bIsTargeting()
   ) -> (Pose) {controlRigAssetReference {blueprintRigClass=&/M_MED_ApronChow_FaceAcc_CtrlRig.M_MED_ApronChow_FaceAcc_CtrlRig_C}}
   n2 = Anim Dynamics - Bone: dyn_hat_l_01!(ComponentPose=Local To Component(LocalPose=@n1), Alpha=1, GravityOverride:Vector="0,0,-980", ExternalForce="0,0,0") -> (Pose) {
    angularDampingOverride=1
    angularSpringConstant=8000
    bAngularSpring=true
    bOverrideAngularDamping=true
    bUseGravityOverride=true
    gravityScale=0.35
    simulationSpace=RootRelative
    worldSpaceMinimumScale=0.2
    boundBone {boneName=dyn_hat_l_01}
    componentAppliedLinearAccClamp=<5000 5000 5000>
    componentLinearAccScale=<0.5 0.5 0.5>
    componentLinearVelScale=<0.5 0.5 0.5>
    gravityOverride {z=-980}
    physicsBodyDefinitions [{
      boundBone {boneName=dyn_hat_l_01}
      boxExtents=<20 30 30>
      constraintSetup {angularLimitsMax=<0 0 2>, angularLimitsMin=<0 0 -8>, angularTarget=<1 0 0>}
      localJointOffset=<2 0 0>
     }]
    relativeSpaceBone {boneName=head}
    simSpaceSettings {simSpaceAngularAlpha=1}
   }
   n3 = Anim Dynamics - Bone: dyn_hat_bk_01!(ComponentPose=@n2, Alpha=1, GravityOverride:Vector="0,0,-980", ExternalForce="0,0,0") -> (Pose) {
    angularDampingOverride=1
    angularSpringConstant=8000
    bAngularSpring=true
    bOverrideAngularDamping=true
    bUseGravityOverride=true
    gravityScale=0.35
    simulationSpace=RootRelative
    worldSpaceMinimumScale=0.2
    boundBone {boneName=dyn_hat_bk_01}
    componentAppliedLinearAccClamp=<5000 5000 5000>
    componentLinearAccScale=<0.5 0.5 0.5>
    componentLinearVelScale=<0.5 0.5 0.5>
    gravityOverride {z=-980}
    physicsBodyDefinitions [{
      boundBone {boneName=dyn_hat_bk_01}
      boxExtents=<20 30 30>
      constraintSetup {angularLimitsMax=<0 0 2>, angularLimitsMin=<0 0 -8>, angularTarget=<1 0 0>}
      localJointOffset=<2 0 0>
     }]
    relativeSpaceBone {boneName=head}
    simSpaceSettings {simSpaceAngularAlpha=1}
   }
   n4 = Anim Dynamics - Bone: dyn_hat_r_01!(ComponentPose=@n3, Alpha=1, GravityOverride:Vector="0,0,-980", ExternalForce="0,0,0") -> (Pose) {
    angularDampingOverride=1
    angularSpringConstant=8000
    bAngularSpring=true
    bOverrideAngularDamping=true
    bUseGravityOverride=true
    gravityScale=0.35
    simulationSpace=RootRelative
    worldSpaceMinimumScale=0.2
    boundBone {boneName=dyn_hat_r_01}
    componentAppliedLinearAccClamp=<5000 5000 5000>
    componentLinearAccScale=<0.5 0.5 0.5>
    componentLinearVelScale=<0.5 0.5 0.5>
    gravityOverride {z=-980}
    physicsBodyDefinitions [{
      boundBone {boneName=dyn_hat_r_01}
      boxExtents=<20 30 30>
      constraintSetup {angularLimitsMax=<0 0 2>, angularLimitsMin=<0 0 -8>, angularTarget=<1 0 0>}
      localJointOffset=<2 0 0>
     }]
    relativeSpaceBone {boneName=head}
    simSpaceSettings {simSpaceAngularAlpha=1}
   }
   n5 = Anim Dynamics - Bone: dyn_hat_fr_01!(ComponentPose=@n4, Alpha=1, GravityOverride:Vector="0,0,-980", ExternalForce="0,0,0") -> (Pose) {
    angularDampingOverride=1
    angularSpringConstant=8000
    bAngularSpring=true
    bOverrideAngularDamping=true
    bUseGravityOverride=true
    gravityScale=0.35
    simulationSpace=RootRelative
    worldSpaceMinimumScale=0.2
    boundBone {boneName=dyn_hat_fr_01}
    componentAppliedLinearAccClamp=<5000 5000 5000>
    componentLinearAccScale=<0.5 0.5 0.5>
    componentLinearVelScale=<0.5 0.5 0.5>
    gravityOverride {z=-980}
    physicsBodyDefinitions [{
      boundBone {boneName=dyn_hat_fr_01}
      boxExtents=<20 30 30>
      constraintSetup {angularLimitsMax=<0 0 2>, angularLimitsMin=<0 0 -8>, angularTarget=<1 0 0>}
      localJointOffset=<2 0 0>
     }]
    relativeSpaceBone {boneName=head}
    simSpaceSettings {simSpaceAngularAlpha=1}
   }
   Output Pose(Result=Component To Local(ComponentPose=@n5)) {layerGroup=None}
  ]}
 EventGraph {nodes [
   Event BlueprintUpdateAnimation()
   > n1 = Sequence()
   > n2 = "Gravity Override State Function (Parameter Asset) (Nativized)"(
    CharacterPartAnimInstance=Self-Reference()
    GameTimeInSeconds=GetGameTimeInSeconds()
    Parameters=&/M_MED_ApronChow_FaceAcc_Parameters
   )
   > Set Gravity Override(Gravity Override=@n2)
   > Set RBAN Settings Out(RBAN Settings Out=@n2.OutRigidBodyAnimNodeParameters)
   n3 = NOT Boolean(A=Get Only_Rifle() -> (Only_Rifle:bool))
   n4 = Get bIsCrouching()
   n5 = Get bIsTargeting()
   Set crouchMoving(execute=@n1.then_1, crouchMoving:double="To Float (Boolean)"(InBool=AND Boolean!(A=@n5, B=@n4, C:bool=@n3) -> (ReturnValue))) -> (Output_Get:double)
   n6 = Get bIsCrouching()
   n7 = Get Only_Rifle()
   n8 = Get bIsTargeting()
   > Set crouchMoving_Rifle(crouchMoving_Rifle:double="To Float (Boolean)"(InBool=AND Boolean(A=@n6, B=@n7, C=@n8))) -> (Output_Get:double)
   n9 = Get bIsCrouching()
   n10 = Get Only_Rifle()
   n11 = Get bIsTargeting()
   > Set Crouch_Rifle(Crouch_Rifle:double="To Float (Boolean)"(InBool=AND Boolean(A=@n9, B=@n10, C=@n11))) -> (Output_Get:double)
   Event OnWeaponChanged()
   > n12 = Cast To FortPlayerAnimInstance(Object=GetAnimInstance(self=GetOwnerSkeletalMeshComponent()))
   > Set pistol(pistol:double=0) -> (Output_Get:double)
   > Set Melee(Melee:double=0) -> (Output_Get:double)
   > Set Rifle(Rifle:double=0) -> (Output_Get:double)
   > n13 = Switch on EFortWeaponCoreAnimation!(execute, Selection:EFortWeaponCoreAnimation Enum=Get WeaponCoreAnim(self=@n12)) -> (
    Melee:Exec
    Pistol:Exec
    Shotgun:Exec
    PaperBlueprint:Exec
    Rifle:Exec
    MeleeOneHand:Exec
    MachinePistol:Exec
    RocketLauncher:Exec
    GrenadeLauncher:Exec
    GoingCommando:Exec
    AssaultRifle:Exec
    TacticalShotgun:Exec
    SniperRifle:Exec
    TrapPlacement:Exec
    ShoulderLauncher:Exec
    AbilityDecoTool:Exec
    Crossbow:Exec
    C4:Exec
    RemoteControl:Exec
    DualWield:Exec
    AR_BullPup:Exec
    AR_ForwardGrip:Exec
    MedPackPaddles:Exec
    SMG_P90:Exec
    AR_DrumGun:Exec
    Consumable_Small:Exec
    Consumable_Large:Exec
    Balloon:Exec
    MountedTurret:Exec
    CreativeTool:Exec
    ExplosiveBow:Exec
    AshtonIndigo:Exec
    AshtonChicago:Exec
    MeleeDualWield:Exec
    Unarmed:Exec
   )
   Set Rifle(
    execute=[
     @n13.AR_BullPup
     @n13.AR_DrumGun
     @n13.AR_ForwardGrip
     @n13.AssaultRifle
     @n13.Crossbow
     @n13.GrenadeLauncher
     @n13.Rifle
     @n13.RocketLauncher
     @n13.SMG_P90
     @n13.Shotgun
     @n13.SniperRifle
     @n13.TacticalShotgun
    ]
    Rifle=1
   )
   > Set Only_Rifle(Only_Rifle:bool=1) -> (Output_Get:bool)
   > Set Only_Pistol(Only_Pistol:bool=false) -> (Output_Get:bool)
   Set pistol(execute=[@n13.MachinePistol, @n13.Pistol], pistol=1)
   > Set Only_Rifle(Only_Rifle=false)
   > Set Only_Pistol(Only_Pistol=true)
   Set Melee(execute=@n13.Melee, Melee=1)
   > Set Only_Rifle(Only_Rifle=false)
   > Set Only_Pistol(Only_Pistol=false)
  ]}
]
variables [
 Crouch_Float {type=boolean}
 Crouch_Rifle {type=number}
 crouchMoving {type=number}
 crouchMoving_Rifle {type=number}
 Gravity Override {type=Vector}
 Melee {type=number}
 Only_Pistol {type=boolean}
 Only_Rifle {type=boolean}
 pistol {type=number}
 RBAN Settings Out {type=FortRigidBodyAnimNodeParameters}
 Rifle {type=number}
]
